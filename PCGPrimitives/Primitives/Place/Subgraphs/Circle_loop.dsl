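; Circle_loop is a PCG loop subgraph that outputs one closed circular spline per iteration, forming rings of growing size.
; The loop index is scaled by increment_Distance and offset by Start, then passes through a further multiply into the sphere radius.
; The circle comes from a Create Points Sphere flattened to a single zero-latitude band of segments.
; A Set op on Rotation.Roll removes the points' vertical orientation before Create Spline closes them into a loop.
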(graph Circle_loop
  (params (increment_Distance :type Double :default 1e+03) (Start :type Double :default 1e+04))
  (bind GetLoopIndex_0 (Get%20Loop%20Index))
  (bind AttributeMathsOp_1 (Attribute%20Maths%20Op :Operation Multiply :InB increment_Distance))
  (bind AttributeMathsOp_2 (Attribute%20Maths%20Op :InB Start))
  (bind AttributeMathsOp_4 (Attribute%20Maths%20Op :Operation Multiply))
  (bind CreatePointsSphere_1 (Create%20Points%20Sphere :LatitudinalEndAngle 0.0 :LatitudinalSegments 1 :LatitudinalStartAngle 0.0 :LongitudinalSegments 64 :Radius 663.616177 :SphereGeneration Segments))
  (bind AttributeMathsOp_3 (Attribute%20Maths%20Op :InputSource1 $Rotation.Roll :Operation Set)) ; removing the vertical nature of the points, \nbecause technically taken from the side of a sphere
  (bind CreateSpline_0 (Create%20Spline :bClosedLoop true :InterpTypeAttribute None))
  (connect GetLoopIndex_0 Out AttributeMathsOp_1 InA)
  (connect AttributeMathsOp_1 Out AttributeMathsOp_2 InA)
  (connect CreatePointsSphere_1 Out AttributeMathsOp_3 InA)
  (connect AttributeMathsOp_2 Out AttributeMathsOp_4 InA)
  (connect AttributeMathsOp_4 Out CreatePointsSphere_1 Radius)
  (connect AttributeMathsOp_3 Out CreateSpline_0 In)
  (output Out CreateSpline_0))
