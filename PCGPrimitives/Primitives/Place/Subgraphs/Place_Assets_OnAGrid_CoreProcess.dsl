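; PCG subgraph that lays out an array of assets on a square grid, one asset per point, and spawns them through a loop subgraph.
; Cell size comes from the largest mesh X/Y extent times a padding ratio, and grid extents from the rounded-up square root of the asset count.
; Surplus points are culled by index, then each receives its asset and a name split from its path.
; Points are offset by an origin and aligned to the actor before spawning.
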(graph Place_Assets_OnAGrid_CoreProcess
  (params (Assets :type (Array (SoftObject /Script/CoreUObject.Object))) (origin_position :type Vector) (padding_ratio :type Double :default 1.5))
  (bind AttributePartition_3 (Attribute%20Partition :PartitionAttributeSelectors [$Index]))
  (bind FilterDataByType_0 (Filter%20Data%20By%20Type :TargetType (struct :Ids [(struct :Struct /Script/PCG.PCGDataTypeInfoPoint)])))
  ;[0.15 0.15 0.15 0.5] Rounded up square of the total assets count.
  (bind GetElementsCount_1 (Get%20Element%20Count :In Assets))
  (bind AttributeCast_0 (Attribute%20Cast))
  (bind AttributeMathsOp_0 (Attribute%20Maths%20Op :Operation Sqrt))
  (bind AttributeMathsOp_8 (Attribute%20Maths%20Op :Operation Ceil))
  ;[0.15 0.15 0.15 0.5] getting max X or Y from the asset list
  (bind AttributePartition_1 (Attribute%20Partition :PartitionAttributeSelectors [$Index] :In Assets))
  (bind AttributeSetToPoint_0 (Attribute%20Set%20To%20Point))
  (bind BoundsFromMesh_1 (Bounds%20From%20Mesh))
  (bind MergePoints_1 (Merge%20Points))
  (bind AttributeMathsOp_7 (Attribute%20Maths%20Op :InputSource1 $Extents :Operation Multiply :InB padding_ratio))
  (bind AttributeReduce_1 (Attribute%20Reduce :InputSource $Extents :Operation Max))
  (bind BreakVectorAttribute_1 (Break%20Vector%20Attribute))
  (bind MergeAttributes_1 (Merge%20Attributes))
  (bind AttributeReduce_3 (Attribute%20Reduce :Operation Max))
  (bind MakeVectorAttribute_1 (Make%20Vector%20Attribute :OutputType Vector))
  ;[0.15 0.15 0.15 0.5] Seeting Grid
  (bind AttributeMathsOp_2 (Attribute%20Maths%20Op :Operation Multiply))
  (bind AttributeMathsOp_4 (Attribute%20Maths%20Op :InputSource1 @Last.Z :Operation Set))
  (bind AttributeMathsOp_5 (Attribute%20Maths%20Op :InputSource1 none :Operation Multiply))
  (bind CreatePointsGrid_1 (Create%20Points%20Grid :CellSize (5e+02 5e+02 5e+01) :CoordinateSpace LocalComponent :GridExtents (249.0 249.0 24.0) :PointPosition CellCorners))
  ;[0.15 0.15 0.15 0.5] One Asset on each point, using index to pick
  (bind GetElementsCount_2 (Get%20Element%20Count :In Assets))
  (bind AttributeFilter_1 (Filter%20Attribute%20Elements :Operator Lesser :TargetAttribute $Index :ThresholdAttribute @Last))
  (bind CopyAttributes_1 (Copy%20Attributes :InputSource $Index :OutputTarget ind :Source Assets :Target Assets))
  (bind ExecuteBlueprint_1 (Execute%20Blueprint :BlueprintElementInstance (class:/PCGPrimitives/Primitives/Shared/PCGBP_SplitString.PCGBP_SplitString_C :OutAttributeName Name :Separator "." :TargetAttribute Assets) :BlueprintElementType /PCGPrimitives/Primitives/Shared/PCGBP_SplitString.PCGBP_SplitString_C))
  (bind AttributePartition_2 (Attribute%20Partition :PartitionAttributeSelectors [ID]))
  (bind AttributeFilter_2 (Filter%20Attribute%20Elements :bUseConstantThreshold true :Operator Equal :TargetAttribute $Index :ThresholdAttribute @Last))
  (bind MergeAttributes_2 (Merge%20Attributes))
  (bind CopyAttributes_2 (Copy%20Attributes :InputSource Name :OutputTarget Name))
  (bind MatchAndSetAttributes_1 (Match%20And%20Set%20Attributes :bMatchAttributes true :InputAttribute $Index :MatchAttribute ind))
  ;[0.15 0.15 0.15 0.5] offset origin
  (bind AttributeMathsOp_3 (Attribute%20Maths%20Op :InputSource1 $Position :InB origin_position))
  ;[0.15 0.15 0.15 0.5] Center to Actor
  (bind GetActorData_3 (Get%20Actor%20Data :Mode GetSinglePoint))
  (bind Subgraph_0 (Subgraph :Graph /PCGPrimitives/Primitives/Transform/Transform_Align.Transform_Align))
  ;[0.15 0.15 0.15 0.5] Spawning
  (bind Loop_1 (Loop :Graph /PCGPrimitives/Primitives/Place/Subgraphs/Place_Assets_OnAGrid_CoreProcess_loop.Place_Assets_OnAGrid_CoreProcess_loop))
  (connect GetElementsCount_1 Out AttributeCast_0 In)
  (connect GetElementsCount_2 Out AttributeFilter_1 Filter)
  (connect CreatePointsGrid_1 Out AttributeFilter_1 In)
  (connect AttributePartition_2 Out AttributeFilter_2 In)
  (connect AttributeCast_0 Out AttributeMathsOp_0 In)
  (connect MakeVectorAttribute_1 Out AttributeMathsOp_2 InA)
  (connect AttributeMathsOp_8 Out AttributeMathsOp_2 InB)
  (connect AttributePartition_3 Out AttributeMathsOp_3 InA)
  (connect AttributeMathsOp_2 Out AttributeMathsOp_4 InA)
  (connect AttributeMathsOp_4 Out AttributeMathsOp_5 InA)
  (connect MergePoints_1 Out AttributeMathsOp_7 InA)
  (connect AttributeMathsOp_0 Out AttributeMathsOp_8 In)
  (connect ExecuteBlueprint_1 Out AttributePartition_2 In)
  (connect MatchAndSetAttributes_1 Out AttributePartition_3 In)
  (connect AttributeMathsOp_7 Out AttributeReduce_1 In)
  (connect MergeAttributes_1 Out AttributeReduce_3 In)
  (connect AttributePartition_1 Out AttributeSetToPoint_0 In)
  (connect AttributeSetToPoint_0 Out BoundsFromMesh_1 In)
  (connect AttributeReduce_1 Out BreakVectorAttribute_1 In)
  (connect MergeAttributes_2 Out CopyAttributes_2 Source)
  (connect CopyAttributes_1 Out CopyAttributes_2 Target)
  (connect MakeVectorAttribute_1 Out CreatePointsGrid_1 CellSize)
  (connect AttributeMathsOp_5 Out CreatePointsGrid_1 GridExtents)
  (connect CopyAttributes_1 Out ExecuteBlueprint_1 In)
  (connect Subgraph_0 Out FilterDataByType_0 In)
  (connect FilterDataByType_0 InsideFilter Loop_1 In)
  (connect AttributeReduce_3 Out MakeVectorAttribute_1 X)
  (connect AttributeReduce_3 Out MakeVectorAttribute_1 Y)
  (connect AttributeFilter_1 InsideFilter MatchAndSetAttributes_1 In)
  (connect CopyAttributes_2 Out MatchAndSetAttributes_1 Match%20Data)
  (connect BreakVectorAttribute_1 X MergeAttributes_1 In)
  (connect BreakVectorAttribute_1 Y MergeAttributes_1 In)
  (connect AttributeFilter_2 OutsideFilter MergeAttributes_2 In)
  (connect BoundsFromMesh_1 Out MergePoints_1 In)
  (connect AttributeMathsOp_3 Out Subgraph_0 PrimaryInput)
  (connect GetActorData_3 Out Subgraph_0 SecondaryInput)
  (output Out Loop_1))
