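; PCG loop-body subgraph for placing assets on a grid that spawns StaticMeshActors from the incoming point data.
; Extract Attribute at Index reads the "name" attribute and feeds it into the Spawn Actor node's ActorLabel override.
; Spawn Actor takes the same input and maps the "Assets" attribute onto each actor's StaticMeshComponent.StaticMesh through a property override.
; Grid positions are not computed here; they arrive with the input.
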(graph Place_Assets_OnAGrid_CoreProcess_loop
  (bind ExtractAttribute_0 (Extract%20Attribute%20at%20Index :bForceOutputAttributeToBeInElementsDomain true :InputSource name))
  (bind SpawnActor_0 (Spawn%20Actor :bAllowTemplateActorEditing true :SpawnedActorPropertyOverrideDescriptions [(struct :InputSource Assets :PropertyTarget "StaticMeshComponent.StaticMesh")] :TemplateActor (class:/Script/Engine.StaticMeshActor) :TemplateActorClass /Script/Engine.StaticMeshActor))
  (connect DefaultInputNode In ExtractAttribute_0 In)
  (connect ExtractAttribute_0 Out SpawnActor_0 ActorLabel)
  (connect DefaultInputNode In SpawnActor_0 In)
  (output Out In))
