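; PCG core subgraph that places points along input splines for the Place_OnSpline primitive.
; It filters polyline data, then uses existing control points or samples by spacing or fixed count, with UseFixedCount driving the sampler mode through a maths op.
; An Offset is applied, then Facing selects a +90, -90 or unchanged yaw.
; With FollowCurve off, an absolute-rotation transform overrides that orientation, discarding spline alignment.
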(graph Place_OnSpline_CoreProcess
  (params (UseExistingControlPoints :type Bool) (Spacing :type Double :default 1.5e+02) (UseFixedCount :type Bool) (Count :type Int) (Facing :type (Enum /PCGPrimitives/Primitives/Place/Subgraphs/EPCG_Facing.EPCG_Facing)) (FollowCurve :type Bool :default true) (Offset :type Vector) (FitToCurve :type Bool) (StartOffset :type Double) (EndOffset :type Double))
  (bind AttributeMathsOp_1 (Attribute%20Maths%20Op :InB UseFixedCount))
  (bind FilterDataByType_0 (Filter%20Data%20By%20Type :TargetType (struct :Ids [(struct :Struct /Script/PCG.PCGDataTypeInfoPolyline)])))
  (bind SplineSampler_1 (Spline%20Sampler :SamplerParams (struct :Mode Distance :bUnbounded true :PointSteepness 1.0) :bFitToCurve FitToCurve :DistanceIncrement Spacing :EndOffset EndOffset :NumSamples Count :StartOffset StartOffset))
  (bind GetSplineControlPoints_1 (Get%20Spline%20Control%20Points))
  (bind Select_0 (Select :bUseInputB-param UseExistingControlPoints))
  (bind TransformPoints_4 (Transform%20Points :OffsetMax-param Offset :OffsetMin-param Offset))
  (bind TransformPoints_1 (Transform%20Points :RotationMax (struct :Yaw 9e+01) :RotationMin (struct :Yaw 9e+01)))
  (bind TransformPoints_2 (Transform%20Points :RotationMax (struct :Yaw -9e+01) :RotationMin (struct :Yaw -9e+01)))
  (bind Select%20%28Multi%29_1 (Select%20%28Multi%29 :EnumSelection (struct :Class /Script/ACLPlugin.ACLCompressionLevel) :IntegerSelection 1 :IntOptions [0 1 2] :IntegerSelection-param Facing))
  (bind TransformPoints_3 (Transform%20Points :bAbsoluteRotation true))
  (bind Select_1 (Select :bUseInputB-param FollowCurve))
  (connect DefaultInputNode In FilterDataByType_0 In)
  (connect FilterDataByType_0 InsideFilter GetSplineControlPoints_1 In)
  (connect SplineSampler_1 Out Select_0 Input%20A)
  (connect GetSplineControlPoints_1 Out Select_0 Input%20B)
  (connect TransformPoints_3 Out Select_1 Input%20A)
  (connect Select%20%28Multi%29_1 Out Select_1 Input%20B)
  (connect TransformPoints_1 Out Select%20%28Multi%29_1 %30)
  (connect TransformPoints_2 Out Select%20%28Multi%29_1 %31)
  (connect TransformPoints_4 Out Select%20%28Multi%29_1 %32)
  (connect AttributeMathsOp_1 Out SplineSampler_1 Mode)
  (connect FilterDataByType_0 InsideFilter SplineSampler_1 Spline)
  (connect TransformPoints_4 Out TransformPoints_1 In)
  (connect TransformPoints_4 Out TransformPoints_2 In)
  (connect Select%20%28Multi%29_1 Out TransformPoints_3 In)
  (connect Select_0 Out TransformPoints_4 In)
  (output Out Select_1))
