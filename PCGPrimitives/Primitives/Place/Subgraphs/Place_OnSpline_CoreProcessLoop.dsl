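; PCG subgraph that places points along input splines, serving as the core process loop of the Place On Spline primitive.
; It keeps polyline data, samples it by distance using Spacing and FitToCurve, then shifts points along Y by Offset.
; Facing selects points yawed plus 90, minus 90 or unrotated.
; FollowCurve keeps that curve-relative rotation; otherwise an absolute-rotation transform resets it, discarding the Facing yaw.
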(graph Place_OnSpline_CoreProcessLoop
  (params (Spacing :type Double :default 1.5e+02) (Facing :type (Enum /PCGPrimitives/Primitives/Place/Subgraphs/EPCG_Facing.EPCG_Facing) :default NewEnumerator1) (FollowCurve :type Bool :default true) (Offset :type Double) (FitToCurve :type Bool))
  (bind FilterDataByType_1 (Filter%20Data%20By%20Type :TargetType (struct :Ids [(struct :Struct /Script/PCG.PCGDataTypeInfoPolyline)])))
  (bind SplineSampler_0 (Spline%20Sampler :SamplerParams (struct :Mode Distance :bUnbounded true :PointSteepness 1.0) :bFitToCurve FitToCurve :DistanceIncrement Spacing))
  (bind MakeVectorAttribute_0 (Make%20Vector%20Attribute :OutputType Vector :Y Offset))
  (bind TransformPoints_3 (Transform%20Points))
  (bind TransformPoints_0 (Transform%20Points :RotationMax (struct :Yaw 9e+01) :RotationMin (struct :Yaw 9e+01)))
  (bind TransformPoints_1 (Transform%20Points :RotationMax (struct :Yaw -9e+01) :RotationMin (struct :Yaw -9e+01)))
  (bind Select%20%28Multi%29_0 (Select%20%28Multi%29 :EnumSelection (struct :Class /Script/ACLPlugin.ACLCompressionLevel) :IntegerSelection 1 :IntOptions [0 1 2] :IntegerSelection-param Facing))
  (bind TransformPoints_2 (Transform%20Points :bAbsoluteRotation true))
  (bind Select_0 (Select :bUseInputB-param FollowCurve))
  (bind MakeVectorAttribute_1 (Make%20Vector%20Attribute :OutputType Vector :Y Offset))
  (connect DefaultInputNode In FilterDataByType_1 In)
  (connect TransformPoints_2 Out Select_0 Input%20A)
  (connect Select%20%28Multi%29_0 Out Select_0 Input%20B)
  (connect TransformPoints_0 Out Select%20%28Multi%29_0 %30)
  (connect TransformPoints_1 Out Select%20%28Multi%29_0 %31)
  (connect TransformPoints_3 Out Select%20%28Multi%29_0 %32)
  (connect FilterDataByType_1 InsideFilter SplineSampler_0 Spline)
  (connect TransformPoints_3 Out TransformPoints_0 In)
  (connect TransformPoints_3 Out TransformPoints_1 In)
  (connect Select%20%28Multi%29_0 Out TransformPoints_2 In)
  (connect SplineSampler_0 Out TransformPoints_3 In)
  (connect MakeVectorAttribute_0 Out TransformPoints_3 OffsetMax)
  (connect MakeVectorAttribute_0 Out TransformPoints_3 OffsetMin)
  (output Out Select_0))
