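; PCG subgraph from the Place primitives that, by its name, handles the core step of placing a point at the center of its input.
; It feeds the input into a Loop node that runs Place_Point_AtCenter_CoreProcess_Loop, passing through the PointExtents parameter, and outputs the loop result.
; The placement logic itself lives in that loop subgraph, so this graph acts as a per-entry wrapper.
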(graph Place_Point_AtCenter_CoreProcess
  (params (PointExtents :type Double :default 1e+02))
  (bind Loop_0 (Loop :Graph /PCGPrimitives/Primitives/Place/Subgraphs/Place_Point_AtCenter_CoreProcess_Loop.Place_Point_AtCenter_CoreProcess_Loop :PointExtents PointExtents))
  (connect DefaultInputNode In Loop_0 In)
  (output Out Loop_0))
