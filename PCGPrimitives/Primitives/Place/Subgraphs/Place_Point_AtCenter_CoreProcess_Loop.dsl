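; PCG loop-body subgraph that places one point at the center of an input polygon or point set, with extents set by PointExtents.
; Polygons have their control points rotated into their longest outer edge's space, combined for an oriented bounds center, then rotated back.
; That point takes the edge's rotation with roll overwritten.
; Point inputs are simply combined without keeping the first point's transform; both branches end in Attribute Noise.
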(graph Place_Point_AtCenter_CoreProcess_Loop
  (params (PointExtents :type Double :default 1e+02))
  (bind FilterDataByType_0 (Filter%20Data%20By%20Type :TargetType (struct :Ids [(struct :Struct /Script/PCG.PCGDataTypeInfoPolygon2D)])))
  (bind CreatePoints_0 (Create%20Points))
  (bind AttributeMathsOp_1 (Attribute%20Maths%20Op :InputSource1 $extents :Operation Set :InB PointExtents))
  (bind GetSplineControlPoints_0 (Get%20Spline%20Control%20Points))
  (bind ToPoint_0 (To%20Point))
  (bind CombinePoints_0 (Combine%20Points))
  (bind CopyPoints_0 (Copy%20Points))
  (bind AttributeMathsOp_0 (Attribute%20Maths%20Op :InputSource1 $Rotation.Roll :Operation Set)) ; Kill Roll
  (bind FilterDataByType_1 (Filter%20Data%20By%20Type :TargetType (struct :Ids [(struct :Struct /Script/PCG.PCGDataTypeInfoPoint)])))
  (bind CombinePoints_1 (Combine%20Points :bUseFirstPointTransform false))
  (bind AttributeMathsOp_2 (Attribute%20Maths%20Op :InputSource1 $extents :Operation Set :InB PointExtents))
  (bind AttributeNoise_1 (Attribute%20Noise))
  ;[0.85 0.55 0.4 0.4] Capture the polygon's longest outer edge as a Transform attribute set.
  ; $LongestOuterSegmentIndex picks the segment, then we filter the spline
  ; control points to that index and extract its Transform. The result is
  ; used below to sample the polygon in its own local space.
  ;;[0.15 0.15 0.15 0.5] FINd and extract LONGEST EDGE
  (bind CopyAttributes_0 (Copy%20Attributes :InputSource @Data.$LongestOuterSegmentIndex :OutputTarget longestSeg))
  (bind CopyAttributes_1 (Copy%20Attributes :InputSource $SegmentLength :OutputTarget lenght))
  (bind CopyAttributes_2 (Copy%20Attributes :InputSource $LocalRotation :OutputTarget localRot))
  (bind AttributeCast_0 (Attribute%20Cast :OutputType Rotator))
  (bind ToPoint_1 (To%20Point))
  (bind AttributeFilter_0 (Filter%20Attribute%20Elements :Operator Equal :TargetAttribute $Index :ThresholdAttribute longestSeg))
  (bind CopyAttributes_3 (Copy%20Attributes :InputSource $Transform :OutputTarget PointTransform))
  (bind PointToAttributeSet_1 (Point%20To%20Attribute%20Set))
  (bind DeleteAttributes_1 (Delete%20Attributes :Operation KeepSelectedAttributes :SelectedAttributes "PointTransform"))
  (bind BreakTransformAttribute_0 (Break%20Transform%20Attribute))
  (bind AttributeCast_1 (Attribute%20Cast :OutputType Rotator))
  ;[0.15 0.15 0.15 0.5] APPLY TRANSFORM
  (bind AttributeVectorOp_3 (Attribute%20Vector%20Op :InputSource1 $Position :Operation InverseTransformDirection))
  ;[0.15 0.15 0.15 0.5] REVERT TRANSFORM
  (bind AttributeVectorOp_2 (Attribute%20Vector%20Op :InputSource1 $Position :Operation TransformDirection))
  (bind CopyAttributes_4 (Copy%20Attributes :InputSource PointTransform :OutputTarget $Rotation))
  (connect CopyAttributes_2 Out AttributeCast_0 In)
  (connect BreakTransformAttribute_0 Rotation AttributeCast_1 In)
  (connect ToPoint_1 Out AttributeFilter_0 In)
  (connect CopyAttributes_4 Out AttributeMathsOp_0 InA)
  (connect CreatePoints_0 Out AttributeMathsOp_1 InA)
  (connect CombinePoints_1 Out AttributeMathsOp_2 InA)
  (connect AttributeMathsOp_0 Out AttributeNoise_1 In)
  (connect AttributeMathsOp_2 Out AttributeNoise_1 In)
  (connect CopyPoints_0 Out AttributeVectorOp_2 In)
  (connect DeleteAttributes_1 Out AttributeVectorOp_2 Transform)
  (connect GetSplineControlPoints_0 Out AttributeVectorOp_3 In)
  (connect DeleteAttributes_1 Out AttributeVectorOp_3 Transform)
  (connect DeleteAttributes_1 Out BreakTransformAttribute_0 In)
  (connect ToPoint_0 Out CombinePoints_0 In)
  (connect FilterDataByType_1 InsideFilter CombinePoints_1 In)
  (connect FilterDataByType_0 InsideFilter CopyAttributes_0 Source)
  (connect FilterDataByType_0 InsideFilter CopyAttributes_0 Target)
  (connect CopyAttributes_0 Out CopyAttributes_1 Source)
  (connect CopyAttributes_0 Out CopyAttributes_1 Target)
  (connect CopyAttributes_1 Out CopyAttributes_2 Source)
  (connect CopyAttributes_1 Out CopyAttributes_2 Target)
  (connect AttributeFilter_0 InsideFilter CopyAttributes_3 Source)
  (connect AttributeFilter_0 InsideFilter CopyAttributes_3 Target)
  (connect AttributeCast_1 Out CopyAttributes_4 Source)
  (connect AttributeVectorOp_2 Out CopyAttributes_4 Target)
  (connect AttributeMathsOp_1 Out CopyPoints_0 Source)
  (connect CombinePoints_0 Out CopyPoints_0 Target)
  (connect PointToAttributeSet_1 Out DeleteAttributes_1 In)
  (connect DefaultInputNode In FilterDataByType_0 In)
  (connect DefaultInputNode In FilterDataByType_1 In)
  (connect FilterDataByType_0 InsideFilter GetSplineControlPoints_0 In)
  (connect CopyAttributes_3 Out PointToAttributeSet_1 In)
  (connect AttributeVectorOp_3 Out ToPoint_0 In)
  (connect AttributeCast_0 Out ToPoint_1 In)
  (output Out AttributeNoise_1))
