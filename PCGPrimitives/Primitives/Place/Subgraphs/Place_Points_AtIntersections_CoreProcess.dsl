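; PCG subgraph that serves as the core process for placing points at spline intersections.
; It filters incoming data down to splines, then feeds them into a Spline Intersection node with a small distance threshold.
; The output contains only the intersection points, along with the indices of the splines involved.
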(graph Place_Points_AtIntersections_CoreProcess
  (bind FilterDataByType_0 (Filter%20Data%20By%20Type :TargetType (struct :Ids [(struct :Struct /Script/PCG.PCGDataTypeInfoSpline)])))
  (bind SplineIntersection_0 (Spline%20Intersection :bOutputSplineIndices true :DistanceThreshold 1.0 :Output IntersectionPointsOnly))
  (connect DefaultInputNode In FilterDataByType_0 In)
  (connect FilterDataByType_0 InsideFilter SplineIntersection_0 In)
  (output Out SplineIntersection_0))
