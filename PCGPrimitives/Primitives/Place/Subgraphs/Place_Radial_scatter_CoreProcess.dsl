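; PCG core subgraph for a radial scatter primitive that places points along concentric circles, copied onto input points when any are supplied.
; A helper spline sampled once per row drives the Circle_loop subgraph, whose resulting splines are sampled at a set arc spacing.
; Density noise sets scale, then bounds are reset and height noise is applied.
; Optional quarter-turn yaw splits points into four re-randomised density bands, each rotated differently, then merges them.
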(graph Place_Radial_scatter_CoreProcess
  (params (increment_Radius_distance :type Double :default 4698.684263) (Start_Radius :type Double :default 57441.371746) (Point_every_x_cm_on_arc :type Double :default 1051.379658) (Number_of_Rows :type Int :default 80) (Scale_min :type Double :default 0.5) (Scale_max :type Double :default 1.5) (Height_variation :type Double) (Seed :type Int) (Random_90_rotation :type Bool :default true))
  (bind CreatePoints_0 (Create%20Points :PointsToCreate [(struct :Transform (struct :Translation (0.0 0.0 1e+03)) :Steepness 1.0) (struct :Steepness 1.0)]))
  (bind CreateSpline_0 (Create%20Spline :bLinear true :InterpTypeAttribute None))
  (bind SplineSampler_0 (Spline%20Sampler :SamplerParams (struct :Mode NumberOfSamples :bUnbounded true :PointSteepness 1.0) :NumSamples Number_of_Rows))
  (bind AttributePartition_1 (Attribute%20Partition :PartitionAttributeSelectors [$Index]))
  (bind Loop_1 (Loop :Graph /PCGPrimitives/Primitives/Place/Subgraphs/Circle_loop.Circle_loop :increment_Distance increment_Radius_distance :Start Start_Radius))
  (bind SplineSampler_1 (Spline%20Sampler :SamplerParams (struct :Mode Distance :DistanceIncrement 2.5e+02 :bFitToCurve true :bUnbounded true :PointSteepness 1.0) :DistanceIncrement Point_every_x_cm_on_arc))
  (bind FilterDataByType_0 (Filter%20Data%20By%20Type :TargetType (struct :Ids [(struct :Struct /Script/PCG.PCGDataTypeInfoPoint)])))
  (bind GetElementsCount_0 (Get%20Element%20Count))
  (bind Branch_1 (Branch))
  (bind CopyPoints_0 (Copy%20Points))
  (bind AttributeNoise_0 (Attribute%20Noise :InputSource $Density))
  (bind AttributeRemap_0 (Attribute%20Remap :InputSource $Density :OutRangeMax 3.0 :OutRangeMin 1.0 :OutRangeMax-param Scale_max :OutRangeMin-param Scale_min))
  (bind CopyAttributes_0 (Copy%20Attributes :InputSource $Density :OutputTarget $Scale))
  (bind BoundsModifier_0 (Bounds%20Modifier :BoundsMin (-1.0 -1.0 -1.0) :Mode Set))
  (bind AttributeNoise_2 (Attribute%20Noise))
  (bind AttributeNoise_3 (Attribute%20Noise :InputSource $Position.Z :NoiseMax 198.52367 :NoiseMax-param Height_variation))
  ;[0.15 0.15 0.15 0.5] Random 90 rotation
  (bind Branch_0 (Branch :bOutputToB-param Random_90_rotation))
  (bind AttributeFilterRange_1 (Filter%20Attribute%20Elements%20by%20Range :MaxThreshold (struct :bUseConstantThreshold true :ThresholdAttribute @Last :bUseSpatialQuery false :AttributeTypes (struct :DoubleValue 0.5)) :MinThreshold (struct :bUseConstantThreshold true :ThresholdAttribute @Last :bUseSpatialQuery false :AttributeTypes (struct :DoubleValue 0.25))))
  (bind TransformPoints_0 (Transform%20Points :RotationMax (struct :Yaw 9e+01) :RotationMin (struct :Yaw 9e+01)))
  (bind AttributeFilterRange_0 (Filter%20Attribute%20Elements%20by%20Range :MaxThreshold (struct :bUseConstantThreshold true :ThresholdAttribute @Last :bUseSpatialQuery false :AttributeTypes (struct :DoubleValue 0.25)) :MinThreshold (struct :bUseConstantThreshold true :ThresholdAttribute @Last :bUseSpatialQuery false) :TargetAttribute $Density))
  (bind AttributeFilterRange_2 (Filter%20Attribute%20Elements%20by%20Range :MaxThreshold (struct :bUseConstantThreshold true :ThresholdAttribute @Last :bUseSpatialQuery false :AttributeTypes (struct :DoubleValue 0.75)) :MinThreshold (struct :bUseConstantThreshold true :ThresholdAttribute @Last :bUseSpatialQuery false :AttributeTypes (struct :DoubleValue 0.5))))
  (bind AttributeFilterRange_3 (Filter%20Attribute%20Elements%20by%20Range :MaxThreshold (struct :bUseConstantThreshold true :ThresholdAttribute @Last :bUseSpatialQuery false :AttributeTypes (struct :DoubleValue 1.0)) :MinThreshold (struct :bUseConstantThreshold true :ThresholdAttribute @Last :bUseSpatialQuery false :AttributeTypes (struct :DoubleValue 0.75))))
  (bind TransformPoints_2 (Transform%20Points :RotationMax (struct :Yaw -9e+01) :RotationMin (struct :Yaw -9e+01)))
  (bind TransformPoints_3 (Transform%20Points :RotationMax (struct :Yaw 1.8e+02) :RotationMin (struct :Yaw 1.8e+02)))
  (bind MergePoints_0 (Merge%20Points))
  (connect Branch_0 Output%20B AttributeFilterRange_0 In)
  (connect Branch_0 Output%20B AttributeFilterRange_1 In)
  (connect Branch_0 Output%20B AttributeFilterRange_2 In)
  (connect Branch_0 Output%20B AttributeFilterRange_3 In)
  (connect Branch_1 Output%20A AttributeNoise_0 In)
  (connect CopyPoints_0 Out AttributeNoise_0 In)
  (connect BoundsModifier_0 Out AttributeNoise_2 In)
  (connect AttributeNoise_2 Out AttributeNoise_3 In)
  (connect SplineSampler_0 Out AttributePartition_1 In)
  (connect AttributeNoise_0 Out AttributeRemap_0 In)
  (connect CopyAttributes_0 Out BoundsModifier_0 In)
  (connect AttributeNoise_3 Out Branch_0 In)
  (connect GetElementsCount_0 Out Branch_1 bOutputToB)
  (connect SplineSampler_1 Out Branch_1 In)
  (connect AttributeRemap_0 Out CopyAttributes_0 Source)
  (connect AttributeRemap_0 Out CopyAttributes_0 Target)
  (connect Branch_1 Output%20B CopyPoints_0 Source)
  (connect FilterDataByType_0 InsideFilter CopyPoints_0 Target)
  (connect CreatePoints_0 Out CreateSpline_0 In)
  (connect DefaultInputNode In FilterDataByType_0 In)
  (connect FilterDataByType_0 InsideFilter GetElementsCount_0 In)
  (connect AttributePartition_1 Out Loop_1 In)
  (connect AttributeFilterRange_0 InsideFilter MergePoints_0 In)
  (connect TransformPoints_0 Out MergePoints_0 In)
  (connect TransformPoints_2 Out MergePoints_0 In)
  (connect TransformPoints_3 Out MergePoints_0 In)
  (connect CreateSpline_0 Out SplineSampler_0 Spline)
  (connect Loop_1 Out SplineSampler_1 Spline)
  (connect AttributeFilterRange_1 InsideFilter TransformPoints_0 In)
  (connect AttributeFilterRange_2 InsideFilter TransformPoints_2 In)
  (connect AttributeFilterRange_3 InsideFilter TransformPoints_3 In)
  (output Out Branch_0.Output%20A)
  (output Out MergePoints_0))
